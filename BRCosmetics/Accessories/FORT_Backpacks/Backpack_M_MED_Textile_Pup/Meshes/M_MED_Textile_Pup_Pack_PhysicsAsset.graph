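# Physics asset for the M_MED_Textile_Pup backpack cosmetic, simulating its handle, zippers and hanging pieces.
# Small capsule bodies sit on the dynamic bones, while two boxes on the pack bone provide collision volumes for the pack body.
# Angular-limited constraints attach the handle and three zipper bones to the pack, and chain hanging pieces from each zipper, with a two-link chain on the front.
# Twist is held very tight, so motion is mostly swing.
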PhysicsAsset /BRCosmetics/Accessories/FORT_Backpacks/Backpack_M_MED_Textile_Pup/Meshes/M_MED_Textile_Pup_Pack_PhysicsAsset
bodies {
 dyn_handle_MSTR [{
   shapeType=Capsule
   length=3.0117
   radius=1.8753
   shapeName=dyn_handle_MSTR_capsule
   center=<1.6911 -0.1825 -0.25>
   rotation=<rot 2.4536 -105.007 2.0101>
  }]
 dyn_hang_bk_r [{
   shapeType=Capsule
   length=3.7777
   radius=0.5974
   shapeName=dyn_hang_bk_r_capsule
   center=<2.2009 0.0576 -0.0393>
   rotation=<rot 0 -91.5759 -89.7319>
  }]
 dyn_hang_fr_02_l [{
   shapeType=Capsule
   length=1.7519
   radius=0.55
   shapeName=dyn_hang_fr_02_l_capsule
   center=<1.0979 0.0138 -0.0015>
   rotation=<rot 12.3752 90.9608 -91.1576>
  }]
 dyn_hang_fr_l [{
   shapeType=Capsule
   length=1
   radius=0.5
   shapeName=dyn_hang_fr_l_capsule
   center=<0.888 -0.0195 0.0426>
   rotation=<rot -12.485 -90.6475 -87.6509>
  }]
 dyn_hang_fr_r [{shapeType=Capsule, length=4.25, radius=0.5, shapeName=dyn_hang_fr_r_capsule, center=<2.1523 0 0>, rotation=<rot 90 -0.0099 -0.0099>}]
 dyn_zipper_bk_r_MSTR [{
   shapeType=Capsule
   length=1.76
   radius=0.5
   shapeName=dyn_zipper_bk_r_MSTR_capsule
   center=<1.3355 0.0622 0.4178>
   rotation=<rot 0 -90.5147 -84.092>
  }]
 dyn_zipper_fr_MSTR [{
   shapeType=Capsule
   length=1.25
   radius=0.5
   shapeName=dyn_zipper_fr_MSTR_capsule
   center=<0.0671 0.0556 0.1391>
   rotation=<rot 1.7398 -95.6544 -90.6326>
  }]
 dyn_zipper_fr_r_MSTR [{
   shapeType=Capsule
   length=1.75
   radius=0.4
   shapeName=dyn_zipper_fr_r_MSTR_capsule
   center=<1.0614 0.094 0.075>
   rotation=<rot -67.729 -115.5141 117.2833>
  }]
 pack [
  {
   shapeType=Box
   extentX=7.8134
   extentY=3.0712
   extentZ=10.3307
   shapeName=right_box
   center=<10.1079 -5.8295 -9.995>
   rotation=<rot 2.458 -87.728 -7.8622>
  }
  {
   shapeType=Box
   extentX=5
   extentY=4
   extentZ=8.5
   shapeName=front_box
   center=<-6.9882 -16.1737 -15.7217>
   rotation=<rot -1.5875 -16.6864 12.5912>
  }
 ]
}
constraints [
 {bone1Name=dyn_zipper_fr_r_MSTR, bone2Name=pack, swing1LimitDegrees=23, swing2LimitDegrees=1.35, twistLimitDegrees=2}
 {bone1Name=dyn_hang_fr_r, bone2Name=dyn_zipper_fr_r_MSTR, swing1LimitDegrees=25, swing2LimitDegrees=20, twistLimitDegrees=1}
 {bone1Name=dyn_zipper_bk_r_MSTR, bone2Name=pack, swing1LimitDegrees=28, swing2LimitDegrees=1.35, twistLimitDegrees=1}
 {bone1Name=dyn_hang_bk_r, bone2Name=dyn_zipper_bk_r_MSTR, swing1LimitDegrees=25, swing2LimitDegrees=20, twistLimitDegrees=1}
 {bone1Name=dyn_zipper_fr_MSTR, bone2Name=pack, swing1LimitDegrees=2, swing2LimitDegrees=75, twistLimitDegrees=1}
 {bone1Name=dyn_hang_fr_l, bone2Name=dyn_zipper_fr_MSTR, swing1LimitDegrees=45, swing2LimitDegrees=60, twistLimitDegrees=1}
 {bone1Name=dyn_hang_fr_02_l, bone2Name=dyn_hang_fr_l, swing1LimitDegrees=45, swing2LimitDegrees=55, twistLimitDegrees=1}
 {bone1Name=dyn_handle_MSTR, bone2Name=pack, swing1LimitDegrees=25, swing2LimitDegrees=10, twistLimitDegrees=2}
]
